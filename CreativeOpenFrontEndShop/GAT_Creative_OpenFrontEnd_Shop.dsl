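; Gameplay ability that opens the Creative standalone item shop frontend, with the parent class handling setup, close and cooldown.
; On activation it waits briefly as a Stasis bug workaround, then the server puts the player pawn in no-movement stasis and damage immunity, reverting both on end.
; When the frontend loads on a machine running cosmetic events, it reads any store deep link from the instigating object and navigates the shop screen to that offer.
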(event Ability|EventActivateAbilityFromEvent
  (params (EventData :type Gameplay%20Event%20Data%20Structure%20%28by%20ref%29))
  ;[1 1 1] On Activate, our parent will asynchronously set up the store frontend. We will apply invulnerability and stasis to the player to prevent issues.
  (|CallParentFunction EventData)
  ;[0.9819 1 0.2169] Delay is required before setting our player effects to avoid an issue with Stasis not being compatible with an ActivateAbilityFromEvent.  Remove this once FORT-546762 has been resolved
  (Utilities|FlowControl|Delay 0.3)
  (CallFunction|SetPlayerStasis true)
  (CallFunction|SetPlayerInvulnerability true))

(event Ability|EventOnEndAbility
  (params (bWasCancelled :type Boolean))
  ;[1 1 1] On End, our parent will apply cooldown and handle close. We will roll back the invulnerability and stasis states.
  (|CallParentFunction bWasCancelled)
  (CallFunction|SetPlayerStasis false)
  (CallFunction|SetPlayerInvulnerability false))

(fn FrontendLoaded
  (params (Loaded%20Widget :type Fort%20Standalone%20Frontend%20Object%20Reference))
  (bind _output_get (LocalVariables|SetLoadedWidget Loaded%20Widget))
  (|CallParentFunction _output_get)
  (Networking|CanExecuteCosmeticEvents
    (:True
      (bind _instigatingobject (Variables|Default|GetInstigatingObject)
        (:then
          ;[1 0 0.8757 0.25] Deep Linking to Store
          (bind (_has_link _deep_link) (CreativeFrontEnd|Store|GetStoreDeepLink(Message) _instigatingobject))
          (if _has_link
            (Development|PrintString (Utilities|String|Append "Store Wants to Deep Link To: " _deep_link) true true "(R=0.000000,G=0.660000,B=1.000000,A=1.000000)" 5e+01)
            (bind _asfort_standalone_item_shop (Utilities|Casting|CastToFortStandaloneItemShop (LocalVariables|GetLoadedWidget)))
            (Class|FortMPItemShopScreen|OnNavigatetoTab (Variables|/Script/FortniteUI.FortStandaloneItemShop|GetActivatableContent_MPItemShop _asfort_standalone_item_shop) _deep_link "NavigateToOffer")))
        (:else)))))

(fn GetPlayerPawn
  ;[1 1 1] Attempts to cast the avatar to a FortPlayerPawn
  (bind _asfort_player_pawn (Utilities|Casting|CastToFortPlayerPawn (Ability|GetAvatarActorfromActorInfo)))
  (return _asfort_player_pawn))

(fn SetPlayerInvulnerability
  (params (IsInvulnerable :type Boolean))
  ;[1 1 1] Sets or unsets Invulnerability on the interacting player pawn
  (Ability|HasAuthority
    (:True
      (bind _playerpawn (CallFunction|GetPlayerPawn))
      (if IsInvulnerable
        (Creative|GameplayEffects|SimpleApplyGameplayEffectToActor "/Game/Abilities/NPC/Generic/GE_DamageImmune.GE_DamageImmune_C" _playerpawn)
        (else
          (bind _returnvalue (Fortnite|Utilities|AttemptGetPlayerStateforPawn _playerpawn))
          (GameplayEffects|RemoveActiveGameplayEffectbySourceEffect (Variables|/Script/FortniteGame.FortPlayerState|GetAbilitySystemComponent _returnvalue) "/Game/Abilities/NPC/Generic/GE_DamageImmune.GE_DamageImmune_C" 0 1))))))

(fn SetPlayerStasis
  (params (IsInStasis :type Boolean))
  ;[1 1 1] Sets or unsets Stasis on the interacting player pawn in order to prevent knockback during shopping or misuse of the device.
  (Ability|HasAuthority
    (:True
      (bind _playerpawn (CallFunction|GetPlayerPawn))
      (if IsInStasis
        (Stasis|AddStasisMode _playerpawn "NoMovementOrFalling" 0 "(TagName=\"Creative.Event.OpenFrontEnd.Store\")" "(GameplayTags=((TagName=\"Gameplay.Status.Athena.Stasis.CursorToggleAllowed\")))")
        (else
          (Stasis|RemoveStasisMode _playerpawn "(TagName=\"Creative.Event.OpenFrontEnd.Store\")" false))))))
